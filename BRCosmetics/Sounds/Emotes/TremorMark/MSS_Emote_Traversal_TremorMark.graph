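# Stereo looping MetaSound Source for the TremorMark traversal emote, with its volume driven by player movement speed.
# A Wave Player starts the traversal loop on play, and both channels are multiplied by a gain from Movement.Speed, range-mapped and smoothed through InterpTo.
# Speed azimuth and elevation inputs are unused, and a PrintFloat debug patch sits unconnected.
# It uses emote attenuation, the music emote sound class and a concurrency override.
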MetaSoundSource /BRCosmetics/Sounds/Emotes/TremorMark/MSS_Emote_Traversal_TremorMark
inputs [
 Movement.Speed {type=Float}
 Movement.SpeedAzimuth {type=Float}
 Movement.SpeedElevation {type=Float}
 Sounds {type=WaveAsset}
 UE.Source.OnPlay {type=Trigger}
]
interfaces [Fort.Movement:1.0, UE.OutputFormat.Stereo:1.0, UE.Source:1.1]
outputs [
 "UE.OutputFormat.Stereo.Audio:0" {type=Audio}
 "UE.OutputFormat.Stereo.Audio:1" {type=Audio}
]
pages [Default {
  input_defaults {
   Movement.Speed=20
   Movement.SpeedAzimuth=0
   Movement.SpeedElevation=0
   Sounds=&/IP_Emote_TremorMark_Loop_Traversal
   UE.Source.OnPlay=null
  }
  nodes [
   InitVariable.WaveAsset:Array()
   n1 = UE.InterpTo:Audio(Target=MapRange.MapRange:Float(In=^Movement.Speed) {also {defaults {In Range B=315}}}.Out Value) {also {defaults {Interp Time=0.3}}}
   WavePlayer_0 = UE.Wave Player:Stereo(Wave Asset=^Sounds, Play=^UE.Source.OnPlay) {also {defaults {Loop="<MetaSoundBuilderResult.SUCCEEDED: 0>"}, version [1,1]}}
   Multiply_0 = UE.Multiply:Audio by Float(AdditionalOperands=@n1.Value, PrimaryOperand=@WavePlayer_0.Out Right)
   Multiply_1 = UE.Multiply:Audio by Float(AdditionalOperands=@n1.Value, PrimaryOperand=@WavePlayer_0.Out Left)
   /FortMetasound/Patches/PrintFloat() {also {defaults {Label=***}}}
  ]
  links [{dst="@UE.OutputFormat.Stereo.Audio:1", src=Multiply_0.Out}, {dst="@UE.OutputFormat.Stereo.Audio:0", src=Multiply_1.Out}]
 }]
properties {
 assetImportData=&self:AssetImportData
 attenuationSettings=&Fort_Emotes_Athena_Att
 bLooping=true
 bOverrideConcurrency=true
 duration=10000
 maxDistance=1600
 numChannels=2
 outputFormat=Stereo
 sampleRate=48000
 soundClassObject=&Fort_Music_Emote
 concurrencyOverrides {maxCount=3}
}
